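# Cosmetic material function that layers tiling sequins onto incoming material attributes, bypassable through the UseSequins static switch.
# Rotated, tiled UVs sample offset, roughness, normal and influence textures, and two colors blend into the sequin tint.
# Mask-shifted falloff, panned noise waves and influence controls vary color and normals, with per-attribute blends against the base material.
# A ray-tracing quality switch scales metalness for GI, and falloff and noise masks are also output.
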MaterialFunction /Game/Global/MaterialFunctions/Cosmetics/Sequins/MF_Sequins
graph {expressions {
  MaterialExpressionAdd_11 {
   class=MaterialExpressionAdd
   inputs [
    A {from=MaterialExpressionAppendVector_0}
    B {from=MaterialExpressionNamedRerouteUsage_51}
   ]
   properties {constB=0}
  }
  MaterialExpressionAdd_3 {
   class=MaterialExpressionAdd
   inputs [
    A {from=MaterialExpressionNamedRerouteUsage_13}
    B {from=MaterialExpressionFunctionInput_36}
   ]
   properties {constB=0.5}
  }
  MaterialExpressionAdd_6 {
   class=MaterialExpressionAdd
   inputs [
    A {from=MaterialExpressionMultiply_19}
    B {from=MaterialExpressionNamedRerouteUsage_20}
   ]
  }
  MaterialExpressionAppendVector_0 {
   class=MaterialExpressionAppendVector
   inputs [
    A {from=MaterialExpressionMultiply_6, output=""}
    B {from=MaterialExpressionComponentMask_2, output=""}
   ]
  }
  MaterialExpressionComponentMask_1 {class=MaterialExpressionComponentMask, inputs [None {from=MaterialExpressionNamedRerouteUsage_40}], properties {g=true, r=true}}
  MaterialExpressionComponentMask_2 {class=MaterialExpressionComponentMask, inputs [None {from=MaterialExpressionNamedRerouteUsage_40}], properties {b=true}}
  MaterialExpressionComponentMask_3 {class=MaterialExpressionComponentMask, inputs [None {from=MaterialExpressionNamedRerouteUsage_54}], properties {g=true, r=true}}
  MaterialExpressionConstant2Vector_0 {class=MaterialExpressionConstant2Vector, properties {g=1}}
  MaterialExpressionConstant2Vector_1 {class=MaterialExpressionConstant2Vector, properties {g=1, r=2}}
  MaterialExpressionConstant3Vector_0 {class=MaterialExpressionConstant3Vector, properties {constant=<rgba 0 0 1 0>}}
  MaterialExpressionConstant3Vector_3 {class=MaterialExpressionConstant3Vector, properties {constant=<rgba 1 1 0 0>}}
  MaterialExpressionConstant4Vector_2 {class=MaterialExpressionConstant4Vector, inputs [], properties {constant=<rgba 0 1 0 0.5>}}
  MaterialExpressionConstant_0 {class=MaterialExpressionConstant}
  MaterialExpressionConstant_10 {class=MaterialExpressionConstant, properties {r=0.5}}
  MaterialExpressionConstant_12 {class=MaterialExpressionConstant, properties {r=0.2}}
  MaterialExpressionConstant_13 {class=MaterialExpressionConstant, properties {r=-0.3}}
  MaterialExpressionConstant_15 {class=MaterialExpressionConstant, properties {r=0.3}}
  MaterialExpressionConstant_19 {class=MaterialExpressionConstant, properties {r=0.1}}
  MaterialExpressionConvert_0 {
   class=MaterialExpressionConvert
   inputs [
    None {from=""}
    None {from=MaterialExpressionLinearInterpolate_9}
    None {from=""}
   ]
   properties {
    convertInputs [
     {type=Scalar, expressionInput {expression=null, inputName=None, mask=0, maskA=0, maskB=0, maskG=0, maskR=0, outputIndex=0}}
     {
      type=Scalar
      expressionInput {
       expression=&self:MaterialExpressionLinearInterpolate_9
       inputName=None
       mask=0
       maskA=0
       maskB=0
       maskG=0
       maskR=0
       outputIndex=0
      }
     }
     {type=Scalar, expressionInput {expression=null, inputName=None, mask=0, maskA=0, maskB=0, maskG=0, maskR=0, outputIndex=0}}
    ]
    convertOutputs [{type=Vector3}]
   }
  }
  MaterialExpressionConvert_1 {
   class=MaterialExpressionConvert
   inputs [None {from=MaterialExpressionFunctionInput_42}]
   properties {
    convertInputs [{
      type=Vector4
      defaultValue=<rgba 0 0 0 1>
      expressionInput {
       expression=&self:MaterialExpressionFunctionInput_42
       inputName=None
       mask=0
       maskA=0
       maskB=0
       maskG=0
       maskR=0
       outputIndex=0
      }
     }]
    convertOutputs [{type=Scalar}, {type=Scalar}, {type=Scalar}, {type=Scalar}]
   }
  }
  MaterialExpressionDivide_0 {
   class=MaterialExpressionDivide
   inputs [
    A {from=MaterialExpressionNamedRerouteUsage_55}
    B {from=""}
   ]
   properties {constB=10}
  }
  MaterialExpressionFunctionInput_0 {
   class=MaterialExpressionFunctionInput
   inputs [Preview {from="", output=""}]
   properties {bUsePreviewValueAsDefault=true, inputName=Metalness_GIBlendAmount, inputType=FunctionInput_Scalar, sortPriority=30}
  }
  MaterialExpressionFunctionInput_1 {
   class=MaterialExpressionFunctionInput
   inputs [Preview {from="", output=""}]
   properties {bUsePreviewValueAsDefault=true, inputName=Fresnel, inputType=FunctionInput_Scalar, sortPriority=41, previewValue=<1 0 0 1>}
  }
  MaterialExpressionFunctionInput_10 {
   class=MaterialExpressionFunctionInput
   inputs [Preview {from="", output=""}]
   properties {bUsePreviewValueAsDefault=true, inputName=BaseColor_Brightness, inputType=FunctionInput_Scalar, sortPriority=25, previewValue=<1.5 0 0 1>}
  }
  MaterialExpressionFunctionInput_16 {
   class=MaterialExpressionFunctionInput
   inputs [Preview {from=MaterialExpressionTextureCoordinate_0, output=""}]
   properties {bUsePreviewValueAsDefault=true, inputName=UV, inputType=FunctionInput_Vector2, sortPriority=2, previewValue=<1 1 0 1>}
  }
  MaterialExpressionFunctionInput_17 {
   class=MaterialExpressionFunctionInput
   inputs [Preview {from="", output=""}]
   properties {bUsePreviewValueAsDefault=true, inputName=OffsetMap, inputType=FunctionInput_Texture2D, sortPriority=11}
  }
  MaterialExpressionFunctionInput_18 {
   class=MaterialExpressionFunctionInput
   inputs [Preview {from=MaterialExpressionConstant_10, output=""}]
   properties {bUsePreviewValueAsDefault=true, inputName=Normal_InfluencedDirectionG, inputType=FunctionInput_Scalar, sortPriority=32}
  }
  MaterialExpressionFunctionInput_2 {
   class=MaterialExpressionFunctionInput
   inputs [Preview {from="", output=""}]
   properties {inputName=Roughness, inputType=FunctionInput_Texture2D, sortPriority=12}
  }
  MaterialExpressionFunctionInput_20 {
   class=MaterialExpressionFunctionInput
   inputs [Preview {from="", output=""}]
   properties {inputName=Normal, inputType=FunctionInput_Texture2D, sortPriority=13}
  }
  MaterialExpressionFunctionInput_21 {
   class=MaterialExpressionFunctionInput
   inputs [Preview {from="", output=""}]
   properties {bUsePreviewValueAsDefault=true, inputName=Normal_Intensity, inputType=FunctionInput_Scalar, sortPriority=31, previewValue=<1 0 0 1>}
  }
  MaterialExpressionFunctionInput_22 {
   class=MaterialExpressionFunctionInput
   inputs [Preview {from="", output=""}]
   properties {bUsePreviewValueAsDefault=true, inputName=TextureMipBias, inputType=FunctionInput_Scalar, sortPriority=11, previewValue=<-1 0 0 1>}
  }
  MaterialExpressionFunctionInput_25 {
   class=MaterialExpressionFunctionInput
   inputs [Preview {from="", output=""}]
   properties {bUsePreviewValueAsDefault=true, inputName=IN_MA, inputType=FunctionInput_MaterialAttributes, sortPriority=1}
  }
  MaterialExpressionFunctionInput_26 {
   class=MaterialExpressionFunctionInput
   inputs [Preview {from="", output=""}]
   properties {bUsePreviewValueAsDefault=true, inputName=-----CoreFeature-----, inputType=FunctionInput_Scalar, sortPriority=20, previewValue=<0.01 0 0 1>}
  }
  MaterialExpressionFunctionInput_27 {
   class=MaterialExpressionFunctionInput
   inputs [Preview {from="", output=""}]
   properties {bUsePreviewValueAsDefault=true, inputName=-----Canvas-----, inputType=FunctionInput_Scalar, previewValue=<0.01 0 0 1>}
  }
  MaterialExpressionFunctionInput_28 {
   class=MaterialExpressionFunctionInput
   inputs [Preview {from="", output=""}]
   properties {bUsePreviewValueAsDefault=true, inputName=Roughness_Min, inputType=FunctionInput_Scalar, sortPriority=28, previewValue=<0 0 0 0>}
  }
  MaterialExpressionFunctionInput_29 {
   class=MaterialExpressionFunctionInput
   inputs [Preview {from="", output=""}]
   properties {bUsePreviewValueAsDefault=true, inputName=Roughness_Max, inputType=FunctionInput_Scalar, sortPriority=29, previewValue=<1 0 0 0>}
  }
  MaterialExpressionFunctionInput_3 {
   class=MaterialExpressionFunctionInput
   inputs [Preview {from="", output=""}]
   properties {bUsePreviewValueAsDefault=true, inputName=BaseColor_Falloff_Amount, inputType=FunctionInput_Scalar, sortPriority=23, previewValue=<1 0 0 1>}
  }
  MaterialExpressionFunctionInput_30 {
   class=MaterialExpressionFunctionInput
   inputs [Preview {from="", output=""}]
   properties {bUsePreviewValueAsDefault=true, inputName=-----ModularInput-----, inputType=FunctionInput_Scalar, sortPriority=40, previewValue=<0.01 0 0 1>}
  }
  MaterialExpressionFunctionInput_32 {
   class=MaterialExpressionFunctionInput
   inputs [Preview {from="", output=""}]
   properties {bUsePreviewValueAsDefault=true, inputName=RotationAngle, inputType=FunctionInput_Scalar, sortPriority=4, previewValue=<0 0 0 0>}
  }
  MaterialExpressionFunctionInput_33 {
   class=MaterialExpressionFunctionInput
   inputs [Preview {from="", output=""}]
   properties {bUsePreviewValueAsDefault=true, inputName=Tiling, inputType=FunctionInput_Scalar, sortPriority=3, previewValue=<1 1 0 1>}
  }
  MaterialExpressionFunctionInput_34 {
   class=MaterialExpressionFunctionInput
   inputs [Preview {from="", output=""}]
   properties {bUsePreviewValueAsDefault=true, inputName=BaseColor_Falloff_Range, inputType=FunctionInput_Scalar, sortPriority=24, previewValue=<1 0 0 1>}
  }
  MaterialExpressionFunctionInput_35 {
   class=MaterialExpressionFunctionInput
   inputs [Preview {from="", output=""}]
   properties {bUsePreviewValueAsDefault=true, inputName=Metalness, inputType=FunctionInput_Scalar, sortPriority=30}
  }
  MaterialExpressionFunctionInput_36 {
   class=MaterialExpressionFunctionInput
   inputs [Preview {from="", output=""}]
   properties {bUsePreviewValueAsDefault=true, inputName=Specular, inputType=FunctionInput_Scalar, sortPriority=27, previewValue=<0.75 0 0 1>}
  }
  MaterialExpressionFunctionInput_37 {
   class=MaterialExpressionFunctionInput
   inputs [Preview {from="", output=""}]
   properties {
    bUsePreviewValueAsDefault=true
    inputName=BaseColor_InfluencedAmount
    inputType=FunctionInput_Scalar
    sortPriority=26
    previewValue=<1 0 0 1>
   }
  }
  MaterialExpressionFunctionInput_38 {
   class=MaterialExpressionFunctionInput
   inputs [Preview {from="", output=""}]
   properties {bUsePreviewValueAsDefault=true, inputName=Normal_InfluencedAmount, inputType=FunctionInput_Scalar, sortPriority=33}
  }
  MaterialExpressionFunctionInput_39 {
   class=MaterialExpressionFunctionInput
   inputs [Preview {from=MaterialExpressionConstant_19, output=""}]
   properties {bUsePreviewValueAsDefault=true, inputName=WaveSpeed, inputType=FunctionInput_Scalar, sortPriority=34, previewValue=<0.1 0 0 1>}
  }
  MaterialExpressionFunctionInput_4 {
   class=MaterialExpressionFunctionInput
   inputs [Preview {from="", output=""}]
   properties {bUsePreviewValueAsDefault=true, inputName=InfluenceMask, inputType=FunctionInput_Texture2D, sortPriority=14}
  }
  MaterialExpressionFunctionInput_40 {
   class=MaterialExpressionFunctionInput
   inputs [Preview {from="", output=""}]
   properties {bUsePreviewValueAsDefault=true, inputName=Normal_WaveAmount, inputType=FunctionInput_Scalar, sortPriority=31}
  }
  MaterialExpressionFunctionInput_41 {
   class=MaterialExpressionFunctionInput
   inputs [Preview {from="", output=""}]
   properties {bUsePreviewValueAsDefault=true, inputName=MA_BaseColor_BlendAmount, inputType=FunctionInput_Scalar, sortPriority=2}
  }
  MaterialExpressionFunctionInput_42 {
   class=MaterialExpressionFunctionInput
   inputs [Preview {from="", output=""}]
   properties {bUsePreviewValueAsDefault=true, inputName=MA_BaseSRMN_BlendAmount, inputType=FunctionInput_Vector4, sortPriority=2}
  }
  MaterialExpressionFunctionInput_5 {
   class=MaterialExpressionFunctionInput
   inputs [Preview {from="", output=""}]
   properties {bUsePreviewValueAsDefault=true, inputName=TechArtMask, inputType=FunctionInput_Scalar, sortPriority=5, previewValue=<1 0 0 1>}
  }
  MaterialExpressionFunctionInput_7 {
   class=MaterialExpressionFunctionInput
   inputs [Preview {from="", output=""}]
   properties {bUsePreviewValueAsDefault=true, inputName=-----Textures-----, inputType=FunctionInput_Scalar, sortPriority=10, previewValue=<0.01 0 0 1>}
  }
  MaterialExpressionFunctionInput_8 {
   class=MaterialExpressionFunctionInput
   inputs [Preview {from="", output=""}]
   properties {bUsePreviewValueAsDefault=true, inputName=BaseColor_01, sortPriority=21, previewValue=<0.5 0.5 0.5 1>}
  }
  MaterialExpressionFunctionInput_9 {
   class=MaterialExpressionFunctionInput
   inputs [Preview {from="", output=""}]
   properties {bUsePreviewValueAsDefault=true, inputName=BaseColor_02, sortPriority=22, previewValue=<0 0 1 1>}
  }
  MaterialExpressionFunctionOutput_4 {
   class=MaterialExpressionFunctionOutput
   inputs [None {from=MaterialExpressionSaturate_0, output=""}]
   properties {outputName=Offseted_FalloffMask, sortPriority=1}
  }
  MaterialExpressionFunctionOutput_7 {
   class=MaterialExpressionFunctionOutput
   inputs [None {from=MaterialExpressionStaticSwitchParameter_1, output=""}]
   properties {outputName=OUT_MA}
  }
  MaterialExpressionFunctionOutput_8 {
   class=MaterialExpressionFunctionOutput
   inputs [None {from=MaterialExpressionNamedRerouteDeclaration_21, output=""}]
   properties {outputName=Offseted_WaveNoise, sortPriority=1}
  }
  MaterialExpressionFunctionOutput_9 {
   class=MaterialExpressionFunctionOutput
   inputs [None {from=MaterialExpressionMultiply_1, output=""}]
   properties {outputName=Offseted_Noise, sortPriority=1}
  }
  MaterialExpressionGetMaterialAttributes_1 {
   class=MaterialExpressionGetMaterialAttributes
   inputs [None {from=MaterialExpressionNamedRerouteDeclaration_10, output=""}]
   properties {attributeGetTypes [
     #1
     #2
     #3
     #4
     #5
    ]}
  }
  MaterialExpressionLinearInterpolate_0 {
   class=MaterialExpressionLinearInterpolate
   inputs [
    A {from=MaterialExpressionFunctionInput_8}
    B {from=MaterialExpressionFunctionInput_9}
    Alpha {from=MaterialExpressionNamedRerouteUsage_56}
   ]
  }
  MaterialExpressionLinearInterpolate_10 {
   class=MaterialExpressionLinearInterpolate
   inputs [
    A {from=MaterialExpressionNamedRerouteUsage_14}
    B {from=MaterialExpressionLinearInterpolate_21}
    Alpha {from=MaterialExpressionNamedRerouteUsage_39}
   ]
  }
  MaterialExpressionLinearInterpolate_11 {
   class=MaterialExpressionLinearInterpolate
   inputs [
    A {from=MaterialExpressionFunctionInput_28}
    B {from=MaterialExpressionFunctionInput_29}
    Alpha {from=MaterialExpressionTextureSample_4, output=R}
   ]
  }
  MaterialExpressionLinearInterpolate_12 {
   class=MaterialExpressionLinearInterpolate
   inputs [
    A {from=MaterialExpressionNamedRerouteUsage_18}
    B {from=MaterialExpressionLinearInterpolate_22}
    Alpha {from=MaterialExpressionNamedRerouteUsage_39}
   ]
  }
  MaterialExpressionLinearInterpolate_13 {
   class=MaterialExpressionLinearInterpolate
   inputs [
    A {from=MaterialExpressionLinearInterpolate_17}
    B {from=MaterialExpressionFunctionInput_18}
    Alpha {from=MaterialExpressionNamedRerouteUsage_49}
   ]
  }
  MaterialExpressionLinearInterpolate_14 {
   class=MaterialExpressionLinearInterpolate
   inputs [
    A {from=MaterialExpressionNamedRerouteUsage_21}
    B {from=MaterialExpressionLinearInterpolate_23}
    Alpha {from=MaterialExpressionNamedRerouteUsage_39}
   ]
  }
  MaterialExpressionLinearInterpolate_15 {
   class=MaterialExpressionLinearInterpolate
   inputs [
    A {from=MaterialExpressionMultiply_35}
    B {from=MaterialExpressionLinearInterpolate_13}
    Alpha {from=MaterialExpressionConstant_15}
   ]
  }
  MaterialExpressionLinearInterpolate_17 {
   class=MaterialExpressionLinearInterpolate
   inputs [
    A {from=MaterialExpressionConstant_12}
    B {from=MaterialExpressionConstant_13}
    Alpha {from=MaterialExpressionNamedRerouteUsage_48}
   ]
   properties {constB=-0.3}
  }
  MaterialExpressionLinearInterpolate_19 {
   class=MaterialExpressionLinearInterpolate
   inputs [
    A {from=MaterialExpressionNamedRerouteUsage_23}
    B {from=MaterialExpressionLinearInterpolate_24}
    Alpha {from=MaterialExpressionNamedRerouteUsage_39}
   ]
  }
  MaterialExpressionLinearInterpolate_20 {
   class=MaterialExpressionLinearInterpolate
   inputs [
    A {from=MaterialExpressionNamedRerouteUsage_26}
    B {from=MaterialExpressionLinearInterpolate_25}
    Alpha {from=MaterialExpressionNamedRerouteUsage_39}
   ]
  }
  MaterialExpressionLinearInterpolate_21 {
   class=MaterialExpressionLinearInterpolate
   inputs [
    A {from=MaterialExpressionNamedRerouteUsage_37}
    B {from=MaterialExpressionNamedRerouteUsage_14}
    Alpha {from=MaterialExpressionFunctionInput_41}
   ]
  }
  MaterialExpressionLinearInterpolate_22 {
   class=MaterialExpressionLinearInterpolate
   inputs [
    A {from=MaterialExpressionNamedRerouteUsage_36}
    B {from=MaterialExpressionNamedRerouteUsage_18}
    Alpha {from=MaterialExpressionConvert_1}
   ]
  }
  MaterialExpressionLinearInterpolate_23 {
   class=MaterialExpressionLinearInterpolate
   inputs [
    A {from=MaterialExpressionMaterialFunctionCall_4, output=Result}
    B {from=MaterialExpressionNamedRerouteUsage_21}
    Alpha {from=MaterialExpressionConvert_1}
   ]
  }
  MaterialExpressionLinearInterpolate_24 {
   class=MaterialExpressionLinearInterpolate
   inputs [
    A {from=MaterialExpressionNamedRerouteUsage_32}
    B {from=MaterialExpressionNamedRerouteUsage_23}
    Alpha {from=MaterialExpressionConvert_1}
   ]
  }
  MaterialExpressionLinearInterpolate_25 {
   class=MaterialExpressionLinearInterpolate
   inputs [
    A {from=MaterialExpressionNamedRerouteUsage_43}
    B {from=MaterialExpressionNamedRerouteUsage_26}
    Alpha {from=MaterialExpressionConvert_1}
   ]
  }
  MaterialExpressionLinearInterpolate_3 {
   class=MaterialExpressionLinearInterpolate
   inputs [
    A {from=MaterialExpressionConstant3Vector_0}
    B {from=MaterialExpressionNamedRerouteUsage_21}
    Alpha {from=MaterialExpressionConvert_1}
   ]
  }
  MaterialExpressionLinearInterpolate_4 {
   class=MaterialExpressionLinearInterpolate
   inputs [
    A {from=MaterialExpressionAppendVector_0}
    B {from=MaterialExpressionAdd_11}
    Alpha {from=MaterialExpressionNamedRerouteUsage_24}
   ]
   properties {constAlpha=1}
  }
  MaterialExpressionLinearInterpolate_7 {
   class=MaterialExpressionLinearInterpolate
   inputs [
    A {from=MaterialExpressionNamedRerouteUsage_3}
    B {from=MaterialExpressionNamedRerouteUsage_52}
    Alpha {from=MaterialExpressionNamedRerouteUsage_19}
   ]
  }
  MaterialExpressionLinearInterpolate_8 {
   class=MaterialExpressionLinearInterpolate
   inputs [
    A {from=MaterialExpressionTextureSample_3, output=RGB}
    B {from=MaterialExpressionMaterialFunctionCall_2, output=Result}
    Alpha {from=MaterialExpressionNamedRerouteUsage_25}
   ]
  }
  MaterialExpressionLinearInterpolate_9 {
   class=MaterialExpressionLinearInterpolate
   inputs [
    A {from=MaterialExpressionMultiply_26}
    B {from=MaterialExpressionFunctionInput_18}
    Alpha {from=MaterialExpressionNamedRerouteUsage_49}
   ]
  }
  MaterialExpressionMaterialFunctionCall_1 {
   class=MaterialExpressionMaterialFunctionCall
   inputs [
    UVs {from=MaterialExpressionFunctionInput_16}
    Rotation Center {from=MaterialExpressionConstant_0}
    "Rotation Angle (0-1)" {from=MaterialExpressionFunctionInput_32}
   ]
   properties {materialFunction=&CustomRotator}
  }
  MaterialExpressionMaterialFunctionCall_2 {
   class=MaterialExpressionMaterialFunctionCall
   inputs [
    Normal {from=MaterialExpressionTextureSample_3, output=RGB}
    Flatness {from=MaterialExpressionNamedRerouteUsage_53}
   ]
   properties {materialFunction=&FlattenNormal}
  }
  MaterialExpressionMaterialFunctionCall_3 {
   class=MaterialExpressionMaterialFunctionCall
   inputs [
    Mask In {from=MaterialExpressionLinearInterpolate_7}
    Amount {from=MaterialExpressionOneMinus_1}
    Range {from=MaterialExpressionFunctionInput_34}
    Start Offset {from=MaterialExpressionConstant2Vector_0, output=R}
    Sharpness {from=MaterialExpressionConstant2Vector_0, output=R}
   ]
   properties {materialFunction=/Game/Global/MaterialFunctions/MF_MaskShifting}
  }
  MaterialExpressionMaterialFunctionCall_4 {
   class=MaterialExpressionMaterialFunctionCall
   inputs [
    BaseNormal {from=MaterialExpressionLinearInterpolate_3}
    AdditionalNormal {from=MaterialExpressionNamedRerouteUsage_12}
   ]
   properties {materialFunction=&BlendAngleCorrectedNormals}
  }
  MaterialExpressionMaterialFunctionCall_5 {
   class=MaterialExpressionMaterialFunctionCall
   inputs [
    Mask In {from=MaterialExpressionAdd_3}
    Amount {from=MaterialExpressionOneMinus_7}
    Range {from=MaterialExpressionConstant4Vector_2, output=G}
    Start Offset {from=MaterialExpressionConstant4Vector_2, output=G}
    Sharpness {from=MaterialExpressionConstant4Vector_2, output=G}
   ]
   properties {materialFunction=&MF_MaskShifting}
  }
  MaterialExpressionMultiply_0 {
   class=MaterialExpressionMultiply
   inputs [
    A {from=MaterialExpressionSaturate_2}
    B {from=MaterialExpressionNamedRerouteUsage_4}
   ]
  }
  MaterialExpressionMultiply_1 {
   class=MaterialExpressionMultiply
   inputs [
    A {from=MaterialExpressionOneMinus_3}
    B {from=MaterialExpressionPower_1}
   ]
  }
  MaterialExpressionMultiply_11 {
   class=MaterialExpressionMultiply
   inputs [
    A {from=MaterialExpressionTextureSample_1, output=R}
    B {from=MaterialExpressionNamedRerouteUsage_7}
   ]
  }
  MaterialExpressionMultiply_19 {
   class=MaterialExpressionMultiply
   inputs [
    A {from=MaterialExpressionTextureCoordinate_2}
    B {from=MaterialExpressionConstant2Vector_1}
   ]
   properties {constB=0.5}
  }
  MaterialExpressionMultiply_2 {
   class=MaterialExpressionMultiply
   inputs [
    A {from=MaterialExpressionFunctionInput_10}
    B {from=MaterialExpressionNamedRerouteUsage_5}
   ]
  }
  MaterialExpressionMultiply_20 {
   class=MaterialExpressionMultiply
   inputs [
    A {from=MaterialExpressionMaterialFunctionCall_5, output=Result}
    B {from=MaterialExpressionFunctionInput_36}
   ]
  }
  MaterialExpressionMultiply_24 {
   class=MaterialExpressionMultiply
   inputs [
    A {from=MaterialExpressionLinearInterpolate_0}
    B {from=MaterialExpressionNamedRerouteUsage_17}
   ]
  }
  MaterialExpressionMultiply_26 {
   class=MaterialExpressionMultiply
   inputs [
    A {from=MaterialExpressionFunctionInput_18}
    B {from=""}
   ]
   properties {constB=-1}
  }
  MaterialExpressionMultiply_3 {
   class=MaterialExpressionMultiply
   inputs [
    A {from=MaterialExpressionMaterialFunctionCall_1, output=Rotated Values}
    B {from=MaterialExpressionFunctionInput_33}
   ]
  }
  MaterialExpressionMultiply_30 {
   class=MaterialExpressionMultiply
   inputs [
    A {from=MaterialExpressionFunctionInput_3}
    B {from=MaterialExpressionNamedRerouteUsage_41}
   ]
  }
  MaterialExpressionMultiply_32 {
   class=MaterialExpressionMultiply
   inputs [
    A {from=MaterialExpressionFunctionInput_35}
    B {from=MaterialExpressionFunctionInput_0}
   ]
  }
  MaterialExpressionMultiply_35 {
   class=MaterialExpressionMultiply
   inputs [
    A {from=MaterialExpressionLinearInterpolate_13}
    B {from=MaterialExpressionOneMinus_2}
   ]
  }
  MaterialExpressionMultiply_4 {
   class=MaterialExpressionMultiply
   inputs [
    A {from=MaterialExpressionMultiply_0}
    B {from=MaterialExpressionNamedRerouteUsage_6}
   ]
  }
  MaterialExpressionMultiply_6 {
   class=MaterialExpressionMultiply
   inputs [
    A {from=MaterialExpressionComponentMask_1}
    B {from=MaterialExpressionFunctionInput_21}
   ]
  }
  MaterialExpressionMultiply_7 {
   class=MaterialExpressionMultiply
   inputs [
    A {from=""}
    B {from=MaterialExpressionConvert_1}
   ]
  }
  MaterialExpressionNamedRerouteDeclaration_0 {
   class=MaterialExpressionNamedRerouteDeclaration
   inputs [None {from=MaterialExpressionMultiply_3}]
   properties {name=SequinsUVs, nodeColor=<rgba 0 1 0.0859 1>}
  }
  MaterialExpressionNamedRerouteDeclaration_1 {
   class=MaterialExpressionNamedRerouteDeclaration
   inputs [None {from=MaterialExpressionMultiply_11}]
   properties {name=SequinsOffset, nodeColor=<rgba 1 0 0.6562 1>}
  }
  MaterialExpressionNamedRerouteDeclaration_10 {
   class=MaterialExpressionNamedRerouteDeclaration
   inputs [None {from=MaterialExpressionFunctionInput_25}]
   properties {name=IN_MA, nodeColor=<rgba 0 1 0.5547 1>}
  }
  MaterialExpressionNamedRerouteDeclaration_16 {
   class=MaterialExpressionNamedRerouteDeclaration
   inputs [None {from=MaterialExpressionFunctionInput_5}]
   properties {name=TechArtMask, nodeColor=<rgba 0 1 0.8125 1>}
  }
  MaterialExpressionNamedRerouteDeclaration_17 {
   class=MaterialExpressionNamedRerouteDeclaration
   inputs [None {from=MaterialExpressionLinearInterpolate_11}]
   properties {name=OUT_Roughness, nodeColor=<rgba 1 0 0.3516 1>}
  }
  MaterialExpressionNamedRerouteDeclaration_19 {
   class=MaterialExpressionNamedRerouteDeclaration
   inputs [None {from=MaterialExpressionTextureSample_10, output=R}]
   properties {name=InfluencerMask, nodeColor=<rgba 0.9688 0 1 1>}
  }
  MaterialExpressionNamedRerouteDeclaration_2 {
   class=MaterialExpressionNamedRerouteDeclaration
   inputs [None {from=MaterialExpressionLinearInterpolate_0}]
   properties {name=SequinColor, nodeColor=<rgba 0.8984 0 1 1>}
  }
  MaterialExpressionNamedRerouteDeclaration_20 {
   class=MaterialExpressionNamedRerouteDeclaration
   inputs [None {from=MaterialExpressionMultiply_2}]
   properties {name=OUT_BaseColor, nodeColor=<rgba 0 0.6719 1 1>}
  }
  MaterialExpressionNamedRerouteDeclaration_21 {
   class=MaterialExpressionNamedRerouteDeclaration
   inputs [None {from=MaterialExpressionMultiply_4}]
   properties {name=WaveNoise, nodeColor=<rgba 0 0.25 1 1>}
  }
  MaterialExpressionNamedRerouteDeclaration_22 {
   class=MaterialExpressionNamedRerouteDeclaration
   inputs [None {from=MaterialExpressionFunctionInput_22}]
   properties {name=TextureMipBias, nodeColor=<rgba 1 1 1 1>}
  }
  MaterialExpressionNamedRerouteDeclaration_23 {
   class=MaterialExpressionNamedRerouteDeclaration
   inputs [None {from=MaterialExpressionLinearInterpolate_8}]
   properties {name=SequinsNormal, nodeColor=<rgba 0.25 0.25 1 1>}
  }
  MaterialExpressionNamedRerouteDeclaration_24 {
   class=MaterialExpressionNamedRerouteDeclaration
   inputs [None {from=MaterialExpressionNamedRerouteDeclaration_19}]
   properties {name=InfluenceOffset, nodeColor=<rgba 0.2891 0 1 1>}
  }
  MaterialExpressionNamedRerouteDeclaration_25 {
   class=MaterialExpressionNamedRerouteDeclaration
   inputs [None {from=MaterialExpressionConvert_0}]
   properties {name=InfluenceNormalDirection, nodeColor=<rgba 0.0312 1 0 1>}
  }
  MaterialExpressionNamedRerouteDeclaration_27 {
   class=MaterialExpressionNamedRerouteDeclaration
   inputs [None {from=MaterialExpressionRayTracingQualitySwitch_1}]
   properties {name=OUT_Metalness, nodeColor=<rgba 0.0781 0 1 1>}
  }
  MaterialExpressionNamedRerouteDeclaration_28 {
   class=MaterialExpressionNamedRerouteDeclaration
   inputs [None {from=MaterialExpressionSaturate_0}]
   properties {name=ViewFalloff_OffsetMask, nodeColor=<rgba 1 0.5391 0 1>}
  }
  MaterialExpressionNamedRerouteDeclaration_3 {
   class=MaterialExpressionNamedRerouteDeclaration
   inputs [None {from=MaterialExpressionOneMinus_6}]
   properties {name=SequinsAlphaMask, nodeColor=<rgba 1 0 0.2344 1>}
  }
  MaterialExpressionNamedRerouteDeclaration_32 {
   class=MaterialExpressionNamedRerouteDeclaration
   inputs [None {from=MaterialExpressionMultiply_20}]
   properties {name=OUT_Specular, nodeColor=<rgba 0 0.6719 1 1>}
  }
  MaterialExpressionNamedRerouteDeclaration_33 {
   class=MaterialExpressionNamedRerouteDeclaration
   inputs [None {from=MaterialExpressionGetMaterialAttributes_1, output=Roughness}]
   properties {name=IN_Roughness, nodeColor=<rgba 0.6172 1 0 1>}
  }
  MaterialExpressionNamedRerouteDeclaration_34 {
   class=MaterialExpressionNamedRerouteDeclaration
   inputs [None {from=MaterialExpressionGetMaterialAttributes_1, output=Normal}]
   properties {name=IN_Normals, nodeColor=<rgba 0 1 0.1562 1>}
  }
  MaterialExpressionNamedRerouteDeclaration_35 {
   class=MaterialExpressionNamedRerouteDeclaration
   inputs [None {from=MaterialExpressionGetMaterialAttributes_1, output=Metallic}]
   properties {name=IN_Metallic, nodeColor=<rgba 1 0.3984 0 1>}
  }
  MaterialExpressionNamedRerouteDeclaration_36 {
   class=MaterialExpressionNamedRerouteDeclaration
   inputs [None {from=MaterialExpressionGetMaterialAttributes_1, output=Specular}]
   properties {name=IN_Specular, nodeColor=<rgba 0 1 0.8828 1>}
  }
  MaterialExpressionNamedRerouteDeclaration_37 {
   class=MaterialExpressionNamedRerouteDeclaration
   inputs [None {from=MaterialExpressionGetMaterialAttributes_1, output=BaseColor}]
   properties {name=IN_BaseColor, nodeColor=<rgba 1 0.3984 0 1>}
  }
  MaterialExpressionNamedRerouteDeclaration_38 {
   class=MaterialExpressionNamedRerouteDeclaration
   inputs [None {from=MaterialExpressionFunctionInput_39}]
   properties {name=WaveSpeed, nodeColor=<rgba 1 0.0234 0 1>}
  }
  MaterialExpressionNamedRerouteDeclaration_4 {
   class=MaterialExpressionNamedRerouteDeclaration
   inputs [None {from=MaterialExpressionFunctionInput_37}]
   properties {name=InfluenceColor_Amount, nodeColor=<rgba 0 1 0.5313 1>}
  }
  MaterialExpressionNamedRerouteDeclaration_5 {
   class=MaterialExpressionNamedRerouteDeclaration
   inputs [None {from=MaterialExpressionFunctionInput_38}]
   properties {name=Normal_InfluencedFixedAmount, nodeColor=<rgba 0 1 0.5312 1>}
  }
  MaterialExpressionNamedRerouteDeclaration_6 {
   class=MaterialExpressionNamedRerouteDeclaration
   inputs [None {from=MaterialExpressionFunctionInput_1}]
   properties {name=Fresnel, nodeColor=<rgba 0.4531 0 1 1>}
  }
  MaterialExpressionNamedRerouteDeclaration_7 {
   class=MaterialExpressionNamedRerouteDeclaration
   inputs [None {from=MaterialExpressionLinearInterpolate_4}]
   properties {name=OUT_Normal, nodeColor=<rgba 0.2891 0 1 1>}
  }
  MaterialExpressionNamedRerouteDeclaration_8 {
   class=MaterialExpressionNamedRerouteDeclaration
   inputs [None {from=MaterialExpressionFunctionInput_40}]
   properties {name=Normal_WaveAmount, nodeColor=<rgba 0 1 0.5312 1>}
  }
  MaterialExpressionNamedRerouteDeclaration_9 {
   class=MaterialExpressionNamedRerouteDeclaration
   inputs [None {from=MaterialExpressionFunctionInput_21}]
   properties {name=Normal_Intensity, nodeColor=<rgba 0.9688 0 1 1>}
  }
  MaterialExpressionNamedRerouteUsage_0 {class=MaterialExpressionNamedRerouteUsage}
  MaterialExpressionNamedRerouteUsage_1 {class=MaterialExpressionNamedRerouteUsage}
  MaterialExpressionNamedRerouteUsage_10 {class=MaterialExpressionNamedRerouteUsage}
  MaterialExpressionNamedRerouteUsage_12 {class=MaterialExpressionNamedRerouteUsage}
  MaterialExpressionNamedRerouteUsage_13 {class=MaterialExpressionNamedRerouteUsage}
  MaterialExpressionNamedRerouteUsage_14 {class=MaterialExpressionNamedRerouteUsage}
  MaterialExpressionNamedRerouteUsage_17 {class=MaterialExpressionNamedRerouteUsage}
  MaterialExpressionNamedRerouteUsage_18 {class=MaterialExpressionNamedRerouteUsage}
  MaterialExpressionNamedRerouteUsage_19 {class=MaterialExpressionNamedRerouteUsage}
  MaterialExpressionNamedRerouteUsage_2 {class=MaterialExpressionNamedRerouteUsage}
  MaterialExpressionNamedRerouteUsage_20 {class=MaterialExpressionNamedRerouteUsage}
  MaterialExpressionNamedRerouteUsage_21 {class=MaterialExpressionNamedRerouteUsage}
  MaterialExpressionNamedRerouteUsage_23 {class=MaterialExpressionNamedRerouteUsage}
  MaterialExpressionNamedRerouteUsage_24 {class=MaterialExpressionNamedRerouteUsage}
  MaterialExpressionNamedRerouteUsage_25 {class=MaterialExpressionNamedRerouteUsage}
  MaterialExpressionNamedRerouteUsage_26 {class=MaterialExpressionNamedRerouteUsage}
  MaterialExpressionNamedRerouteUsage_28 {class=MaterialExpressionNamedRerouteUsage}
  MaterialExpressionNamedRerouteUsage_3 {class=MaterialExpressionNamedRerouteUsage}
  MaterialExpressionNamedRerouteUsage_30 {class=MaterialExpressionNamedRerouteUsage}
  MaterialExpressionNamedRerouteUsage_32 {class=MaterialExpressionNamedRerouteUsage}
  MaterialExpressionNamedRerouteUsage_35 {class=MaterialExpressionNamedRerouteUsage}
  MaterialExpressionNamedRerouteUsage_36 {class=MaterialExpressionNamedRerouteUsage}
  MaterialExpressionNamedRerouteUsage_37 {class=MaterialExpressionNamedRerouteUsage}
  MaterialExpressionNamedRerouteUsage_39 {class=MaterialExpressionNamedRerouteUsage}
  MaterialExpressionNamedRerouteUsage_4 {class=MaterialExpressionNamedRerouteUsage}
  MaterialExpressionNamedRerouteUsage_40 {class=MaterialExpressionNamedRerouteUsage}
  MaterialExpressionNamedRerouteUsage_41 {class=MaterialExpressionNamedRerouteUsage}
  MaterialExpressionNamedRerouteUsage_43 {class=MaterialExpressionNamedRerouteUsage}
  MaterialExpressionNamedRerouteUsage_45 {class=MaterialExpressionNamedRerouteUsage}
  MaterialExpressionNamedRerouteUsage_46 {class=MaterialExpressionNamedRerouteUsage}
  MaterialExpressionNamedRerouteUsage_47 {class=MaterialExpressionNamedRerouteUsage}
  MaterialExpressionNamedRerouteUsage_48 {class=MaterialExpressionNamedRerouteUsage}
  MaterialExpressionNamedRerouteUsage_49 {class=MaterialExpressionNamedRerouteUsage}
  MaterialExpressionNamedRerouteUsage_5 {class=MaterialExpressionNamedRerouteUsage}
  MaterialExpressionNamedRerouteUsage_51 {class=MaterialExpressionNamedRerouteUsage}
  MaterialExpressionNamedRerouteUsage_52 {class=MaterialExpressionNamedRerouteUsage}
  MaterialExpressionNamedRerouteUsage_53 {class=MaterialExpressionNamedRerouteUsage}
  MaterialExpressionNamedRerouteUsage_54 {class=MaterialExpressionNamedRerouteUsage}
  MaterialExpressionNamedRerouteUsage_55 {class=MaterialExpressionNamedRerouteUsage}
  MaterialExpressionNamedRerouteUsage_56 {class=MaterialExpressionNamedRerouteUsage}
  MaterialExpressionNamedRerouteUsage_6 {class=MaterialExpressionNamedRerouteUsage}
  MaterialExpressionNamedRerouteUsage_7 {class=MaterialExpressionNamedRerouteUsage}
  MaterialExpressionOneMinus_1 {class=MaterialExpressionOneMinus, inputs [None {from=MaterialExpressionMultiply_30}]}
  MaterialExpressionOneMinus_2 {class=MaterialExpressionOneMinus, inputs [None {from=MaterialExpressionNamedRerouteUsage_47}]}
  MaterialExpressionOneMinus_3 {class=MaterialExpressionOneMinus, inputs [None {from=MaterialExpressionStep_2}]}
  MaterialExpressionOneMinus_6 {class=MaterialExpressionOneMinus, inputs [None {from=MaterialExpressionStep_1}]}
  MaterialExpressionOneMinus_7 {class=MaterialExpressionOneMinus, inputs [None {from=MaterialExpressionFunctionInput_36}]}
  MaterialExpressionPanner_0 {
   class=MaterialExpressionPanner
   inputs [
    Coordinate {from=MaterialExpressionComponentMask_3}
    Time {from=""}
    Speed {from=MaterialExpressionDivide_0}
   ]
   properties {speedX=0.1, speedY=0.1}
  }
  MaterialExpressionPanner_1 {
   class=MaterialExpressionPanner
   inputs [
    Coordinate {from=MaterialExpressionAdd_6}
    Time {from=""}
    Speed {from=MaterialExpressionNamedRerouteDeclaration_38}
   ]
   properties {speedX=0.1, speedY=0.1}
  }
  MaterialExpressionPower_1 {
   class=MaterialExpressionPower
   inputs [
    Base {from=MaterialExpressionTextureSample_0, output=R}
    Exp {from=""}
   ]
   properties {constExponent=4}
  }
  MaterialExpressionPower_3 {
   class=MaterialExpressionPower
   inputs [
    Base {from=MaterialExpressionTextureSample_11, output=R}
    Exp {from=""}
   ]
  }
  MaterialExpressionRayTracingQualitySwitch_1 {
   class=MaterialExpressionRayTracingQualitySwitch
   inputs [
    Normal {from=MaterialExpressionFunctionInput_35, output=""}
    RayTraced {from=MaterialExpressionMultiply_32, output=""}
   ]
   properties {}
  }
  MaterialExpressionSaturate_0 {class=MaterialExpressionSaturate, inputs [None {from=MaterialExpressionMaterialFunctionCall_3, output=Result}]}
  MaterialExpressionSaturate_2 {class=MaterialExpressionSaturate, inputs [None {from=MaterialExpressionPower_3}]}
  MaterialExpressionSetMaterialAttributes_1 {
   class=MaterialExpressionSetMaterialAttributes
   inputs [
    MaterialAttributes {from=MaterialExpressionGetMaterialAttributes_1, output=MaterialAttributes}
    BaseColor {from=MaterialExpressionLinearInterpolate_10, output=""}
    Specular {from=MaterialExpressionLinearInterpolate_20, output=""}
    Roughness {from=MaterialExpressionLinearInterpolate_12, output=""}
    Metallic {from=MaterialExpressionLinearInterpolate_19, output=""}
    Normal {from=MaterialExpressionLinearInterpolate_14, output=""}
   ]
   properties {attributeSetTypes [
     #1
     #4
     #2
     #5
     #3
    ]}
  }
  MaterialExpressionStaticSwitchParameter_1 {
   class=MaterialExpressionStaticSwitchParameter
   inputs [
    True {from=MaterialExpressionSetMaterialAttributes_1, output=""}
    False {from=MaterialExpressionNamedRerouteUsage_10, output=""}
   ]
   properties {defaultValue=true, group=04_Tech Art, parameterName=UseSequins}
  }
  MaterialExpressionStep_1 {
   class=MaterialExpressionStep
   inputs [
    Y {from=MaterialExpressionTextureSample_1, output=R}
    X {from="", output=""}
   ]
   properties {constX=0}
  }
  MaterialExpressionStep_2 {
   class=MaterialExpressionStep
   inputs [
    Y {from=MaterialExpressionNamedRerouteUsage_54, output=""}
    X {from="", output=""}
   ]
   properties {constX=0}
  }
  MaterialExpressionTextureCoordinate_0 {class=MaterialExpressionTextureCoordinate}
  MaterialExpressionTextureCoordinate_2 {class=MaterialExpressionTextureCoordinate, properties {vTiling=0.5}}
  MaterialExpressionTextureSample_0 {
   class=MaterialExpressionTextureSample
   inputs [
    UVs {from=MaterialExpressionPanner_0}
    Tex {from=""}
    Apply View MipBias {from=""}
   ]
   properties {
    desc=Inner Texture
    samplerSource=SSM_Wrap_WorldGroupSettings
    samplerType=SAMPLERTYPE_Normal
    texture=/Game/Global/Textures/Default/Noise/64x64Noise
   }
  }
  MaterialExpressionTextureSample_1 {
   class=MaterialExpressionTextureSample
   inputs [
    UVs {from=MaterialExpressionNamedRerouteUsage_0}
    Tex {from=MaterialExpressionFunctionInput_17}
    Bias {from=MaterialExpressionNamedRerouteUsage_28}
    Apply View MipBias {from=""}
   ]
   properties {
    mipValueMode=TMVM_MipBias
    samplerSource=SSM_Wrap_WorldGroupSettings
    samplerType=SAMPLERTYPE_LinearGrayscale
    texture=/Game/Global/Textures/Default/Tileable_Patterns/T_SequinTile
   }
  }
  MaterialExpressionTextureSample_10 {
   class=MaterialExpressionTextureSample
   inputs [
    UVs {from=MaterialExpressionNamedRerouteUsage_45}
    Tex {from=MaterialExpressionFunctionInput_4}
    Bias {from=MaterialExpressionNamedRerouteUsage_46}
    Apply View MipBias {from=""}
   ]
   properties {
    mipValueMode=TMVM_MipBias
    samplerSource=SSM_Wrap_WorldGroupSettings
    samplerType=SAMPLERTYPE_LinearGrayscale
    texture=&T_SequinTile
   }
  }
  MaterialExpressionTextureSample_11 {
   class=MaterialExpressionTextureSample
   inputs [
    UVs {from=MaterialExpressionPanner_1}
    Tex {from=""}
    Apply View MipBias {from=""}
   ]
   properties {samplerSource=SSM_Wrap_WorldGroupSettings, samplerType=SAMPLERTYPE_Grayscale, texture=&LowResBlurredNoise}
  }
  MaterialExpressionTextureSample_3 {
   class=MaterialExpressionTextureSample
   inputs [
    UVs {from=MaterialExpressionNamedRerouteUsage_2}
    Tex {from=MaterialExpressionFunctionInput_20}
    Bias {from=MaterialExpressionNamedRerouteUsage_35}
    Apply View MipBias {from=""}
   ]
   properties {
    mipValueMode=TMVM_MipBias
    samplerSource=SSM_Wrap_WorldGroupSettings
    samplerType=SAMPLERTYPE_Normal
    texture=/Game/Global/Textures/Default/Tileable_Patterns/T_SequinTile_N
   }
  }
  MaterialExpressionTextureSample_4 {
   class=MaterialExpressionTextureSample
   inputs [
    UVs {from=MaterialExpressionNamedRerouteUsage_1}
    Tex {from=MaterialExpressionFunctionInput_2}
    Bias {from=MaterialExpressionNamedRerouteUsage_30}
    Apply View MipBias {from=""}
   ]
   properties {
    mipValueMode=TMVM_MipBias
    samplerSource=SSM_Wrap_WorldGroupSettings
    samplerType=SAMPLERTYPE_LinearGrayscale
    texture=/Game/Global/Textures/Default/Tileable_Patterns/T_SequinTile_roughness
   }
  }
 }}
