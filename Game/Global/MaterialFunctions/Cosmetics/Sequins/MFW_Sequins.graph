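# Wrapper material function that exposes the cosmetic sequins effect, MF_Sequins, as "Sequins" group parameters on incoming material attributes.
# Channel selectors split TechArtMask into primary and secondary masks.
# The UseSecondaryColor static switch sums both masks and mask-weights each set of falloff colors, amount, range and color influence.
# It outputs the attributes plus offset noise, wave noise and falloff mask; the Normal_InfluencedAmount and Normal_InfluencedDirectionG parameter names are cross-wired.
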MaterialFunction /Game/Global/MaterialFunctions/Cosmetics/Sequins/MFW_Sequins
graph {nodes [
  n1 = VectorParameter() {
   group=Sequins
   parameterName=Sequins.BaseSRMN_BlendAmount
   sortPriority=4
   channelNames {a=Blend_To_BaseNormal, b=Blend_To_BaseMetalness, g=Blend_To_BaseRoughness, r=Blend_To_BaseSpecular}
   defaultValue=<rgba 0 0 0 0.5>
  }
  n2 = Convert() {
   convertInputs [{
     type=Vector3
     defaultValue=<rgba 1 0 0 1>
     expressionInput {expression=null, inputName=None, mask=0, maskA=0, maskB=0, maskG=0, maskR=0, outputIndex=0}
    }]
   convertOutputs [{type=Vector3}, {type=Vector3}]
  }
  n3 = Step(Y=ComponentMask(None=TextureCoordinate()) {r=true}) {constX=0.5}
  n4 = FunctionInput(Preview=StaticSwitchParameter(True=LinearInterpolate(A=@n2, B=@n2, Alpha=@n3), False=@n2) {group=Sequins, parameterName=UseSecondaryColor, sortPriority=0}) {bUsePreviewValueAsDefault=true, inputName=TechArtMask, sortPriority=2}
  n5 = NamedRerouteDeclaration(None=ChannelMaskParameter(None=@n4) {group=Sequins, parameterName=Sequins.MaskChannel, primitiveDataIndex=1, sortPriority=1}.R) {name=SequinsMask_Main, nodeColor=<rgba 0.5 1 0 1>}
  n6 = ChannelMaskParameter(None=@n4) {
   group=Sequins
   maskChannel=Green
   parameterName=Sequins.MaskChannel_Secondary
   primitiveDataIndex=2
   sortPriority=1
   defaultValue=<rgba 0 1 0 0>
  }
  n7 = Add(A=@n5, B=NamedRerouteDeclaration(None=@n6.R) {name=SequinsMask_Secondary, nodeColor=<rgba 0.5 1 0 1>})
  n8 = TextureObjectParameter() {
   group=Sequins
   parameterName=Sequins.OffsetMap
   samplerType=SAMPLERTYPE_LinearGrayscale
   sortPriority=6
   texture=/Game/Global/Textures/Default/Tileable_Patterns/T_SequinTile
  }
  n9 = TextureObjectParameter() {
   group=Sequins
   parameterName=Sequins.Roughness
   samplerType=SAMPLERTYPE_LinearGrayscale
   sortPriority=7
   texture=/Game/Global/Textures/Default/Tileable_Patterns/T_SequinTile_roughness
  }
  n10 = TextureObjectParameter() {
   group=Sequins
   parameterName=Sequins.Normal
   samplerType=SAMPLERTYPE_Normal
   sortPriority=8
   texture=/Game/Global/Textures/Default/Tileable_Patterns/T_SequinTile_N
  }
  n11 = TextureObjectParameter() {
   group=Sequins
   parameterName=Sequins.InfluenceMask
   samplerType=SAMPLERTYPE_Masks
   sortPriority=9
   texture=&T_BlackMask
  }
  n12 = FunctionInput(Preview=Fresnel(ExponentIn=Constant() {r=0.5}) {baseReflectFraction=0}) {bUsePreviewValueAsDefault=true, inputName=Fresnel, inputType=FunctionInput_Scalar, sortPriority=3, previewValue=<1 0 0 1>}
  n13 = MaterialFunctionCall!(
   -----Canvas-----
   IN_MA=FunctionInput() {bUsePreviewValueAsDefault=true, inputName=IN_MA, inputType=FunctionInput_MaterialAttributes}
   MA_BaseColor_BlendAmount=ScalarParameter() {group=Sequins, parameterName=Sequins.BaseColor_BlendAmount, sortPriority=3}
   MA_BaseSRMN_BlendAmount=@n1.RGBA
   UV=FunctionInput(Preview=TextureCoordinate()) {bUsePreviewValueAsDefault=true, inputName=UVs, inputType=FunctionInput_Vector2, sortPriority=1}
   Tiling=ScalarParameter() {defaultValue=2, group=Sequins, parameterName=Sequins.Tiling, sortPriority=4}
   RotationAngle=ScalarParameter() {group=Sequins, parameterName=Sequins.RotationAngle, sortPriority=5}
   TechArtMask=NamedRerouteDeclaration(None=StaticSwitchParameter(True=@n7, False=@n5) {group=Sequins, parameterName=UseSecondaryColor, sortPriority=0}) {name=SequinsMask, nodeColor=<rgba 1 0.1406 0 1>}
   -----Textures-----
   TextureMipBias=ScalarParameter() {defaultValue=-1, group=Sequins, parameterName=Sequins.TextureMipBias, sortPriority=5}
   OffsetMap=@n8
   Roughness=@n9
   Normal=@n10
   InfluenceMask=@n11
   -----CoreFeature-----
   BaseColor_01=NamedRerouteUsage()
   BaseColor_02=NamedRerouteUsage()
   BaseColor_Falloff_Amount=NamedRerouteUsage()
   BaseColor_Falloff_Range=NamedRerouteUsage()
   BaseColor_Brightness=ScalarParameter() {defaultValue=1, group=Sequins, parameterName=Sequins.ColorBrightness, sortPriority=16}
   BaseColor_InfluencedAmount=NamedRerouteUsage()
   Specular=ScalarParameter() {defaultValue=0.5, group=Sequins, parameterName=Sequins.Specular, sortPriority=18}
   Roughness_Min=ScalarParameter() {group=Sequins, parameterName=Sequins.Roughness_Min, sortPriority=19}
   Roughness_Max=ScalarParameter() {defaultValue=1, group=Sequins, parameterName=Sequins.Roughness_Max, sortPriority=20}
   Metalness=ScalarParameter() {defaultValue=1, group=Sequins, parameterName=Sequins.Metalness, sortPriority=21}
   Metalness_GIBlendAmount=ScalarParameter() {defaultValue=0.75, group=Sequins, parameterName=Sequins.Metalness_GIBlendAmount, sortPriority=22}
   Normal_WaveAmount=ScalarParameter() {defaultValue=1, group=Sequins, parameterName=Sequins.Normal_WaveAmount, sortPriority=24}
   Normal_Intensity=ScalarParameter() {defaultValue=0.8, group=Sequins, parameterName=Sequins.Normal_Intensity, sortPriority=23}
   Normal_InfluencedDirectionG=ScalarParameter() {group=Sequins, parameterName=Sequins.Normal_InfluencedAmount, sortPriority=25}
   Normal_InfluencedAmount=ScalarParameter() {defaultValue=0.6, group=Sequins, parameterName=Sequins.Normal_InfluencedDirectionG, sortPriority=26}
   WaveSpeed=ScalarParameter() {defaultValue=0.2, group=Sequins, parameterName=Sequins.WaveSpeed, sortPriority=27}
   -----ModularInput-----
   Fresnel=@n12
  ) {materialFunction=&/MF_Sequins}
  FunctionOutput(None=@n13.Offseted_Noise) {outputName=Sequins_Offseted_Noise, sortPriority=2}
  FunctionOutput(None=@n13.OUT_MA) {outputName=OUT_MA}
  FunctionOutput(None=@n13.Offseted_WaveNoise) {outputName=Sequins_Offseted_WaveNoise, sortPriority=1}
  FunctionOutput(None=@n13.Offseted_FalloffMask) {outputName=Sequins_Offseted_FalloffMask, sortPriority=3}
  n14 = ScalarParameter() {defaultValue=0.1, group=Sequins, parameterName=Sequins.FalloffColor_Range, sortPriority=15}
  n15 = Multiply(A=@n14, B=NamedRerouteUsage())
  n16 = Multiply(
   A=ScalarParameter() {defaultValue=0.1, group=Sequins, parameterName=Sequins.FalloffColor_Range_Secondary, sortPriority=15}
   B=NamedRerouteUsage()
  )
  NamedRerouteDeclaration(None=StaticSwitchParameter(True=Add(A=@n15, B=@n16), False=@n14) {group=Sequins, parameterName=UseSecondaryColor, sortPriority=0}) {name=FalloffColor_Range, nodeColor=<rgba 0.1016 1 0 1>}
  n17 = VectorParameter() {group=Sequins, parameterName=Sequins.FalloffColor_01, sortPriority=10, defaultValue=<rgba 1 0 0 1>}
  n18 = NamedRerouteUsage()
  n19 = Multiply(A=@n17.RGB, B=@n18)
  n20 = VectorParameter() {group=Sequins, parameterName=Sequins.FalloffColor_01_Secondary, sortPriority=12, defaultValue=<rgba 0 1 0 1>}
  n21 = NamedRerouteUsage()
  n22 = Multiply(A=@n20.RGB, B=@n21)
  NamedRerouteDeclaration(None=StaticSwitchParameter(True=Add(A=@n19, B=@n22), False=@n19) {group=Sequins, parameterName=UseSecondaryColor, sortPriority=0}) {name=BaseColor_01, nodeColor=<rgba 0.1016 1 0 1>}
  n23 = VectorParameter() {group=Sequins, parameterName=Sequins.FalloffColor_02, sortPriority=11, defaultValue=<rgba 0.25 0 0 1>}
  n24 = Multiply(A=@n23.RGB, B=@n18)
  n25 = VectorParameter() {group=Sequins, parameterName=Sequins.FalloffColor_02_Secondary, sortPriority=13, defaultValue=<rgba 0 0.25 0 1>}
  n26 = Multiply(A=@n25.RGB, B=@n21)
  NamedRerouteDeclaration(None=StaticSwitchParameter(True=Add(A=@n24, B=@n26), False=@n24) {group=Sequins, parameterName=UseSecondaryColor, sortPriority=0}) {name=BaseColor_02, nodeColor=<rgba 0.1016 1 0 1>}
  n27 = ScalarParameter() {group=Sequins, parameterName=Sequins.BaseColor_InfluencedAmount, sortPriority=17}
  n28 = Multiply(A=@n27, B=NamedRerouteUsage())
  n29 = Multiply(A=ScalarParameter() {group=Sequins, parameterName=Sequins.BaseColor_InfluencedAmount_Secondary, sortPriority=17}, B=NamedRerouteUsage())
  NamedRerouteDeclaration(None=StaticSwitchParameter(True=Add(A=@n28, B=@n29), False=@n27) {group=Sequins, parameterName=UseSecondaryColor, sortPriority=0}) {name=BaseColor_InfluencedAmount, nodeColor=<rgba 0.3565 1 0 1>}
  n30 = ScalarParameter() {defaultValue=1, group=Sequins, parameterName=Sequins.FalloffColor_Amount, sortPriority=14}
  n31 = Multiply(A=@n30, B=NamedRerouteUsage())
  n32 = Multiply(
   A=ScalarParameter() {defaultValue=1, group=Sequins, parameterName=Sequins.FalloffColor_Amont_Secondary, sortPriority=14}
   B=NamedRerouteUsage()
  )
  NamedRerouteDeclaration(None=StaticSwitchParameter(True=Add(A=@n31, B=@n32), False=@n30) {group=Sequins, parameterName=UseSecondaryColor, sortPriority=0}) {name=FalloffColor_Amount, nodeColor=<rgba 0.1016 1 0 1>}
 ]}
